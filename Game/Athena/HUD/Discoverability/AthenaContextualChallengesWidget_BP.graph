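# HUD widget that shows a list of contextual "Local Quests", with the header switching to "{poi name} Quests" when a location name is supplied.
# Activation with entries present marks the last entry, flags local challenges visible on the HUD context and plays a slide-and-fade intro through SimpleAnimations.
# An optional timer, restartable by outro-delay resets, drives auto-dismissal.
# Finishing plays the reverse animation, whose completion clears the HUD visibility flag.
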WidgetBlueprint /Game/Athena/HUD/Discoverability/AthenaContextualChallengesWidget_BP
animations [Intro {
  binding_count=1
  length_seconds=0.7972
  object=&self:Intro
  sequence {bindings [VerticalBox_0 {
     id=#1
     possessed_class=VerticalBox
     tracks [
      {
       type=MovieSceneFloatTrack
       display_name=Render Opacity
       sections [{type=MovieSceneFloatSection, channels [None {type=MovieSceneScriptingFloatChannel, keys [[0.0553,0],[16,1]]}]}]
      }
      {
       type=MovieScene2DTransformTrack
       sections [{
         type=MovieScene2DTransformSection
         channels [
          Translation.X {type=MovieSceneScriptingFloatChannel, keys [[0,-32],[16,0]]}
          Translation.Y {type=MovieSceneScriptingFloatChannel}
          Angle {type=MovieSceneScriptingFloatChannel}
          Scale.X {type=MovieSceneScriptingFloatChannel}
          Scale.Y {type=MovieSceneScriptingFloatChannel}
          Shear.X {type=MovieSceneScriptingFloatChannel}
          Shear.Y {type=MovieSceneScriptingFloatChannel}
         ]
        }]
      }
     ]
    }]}
 }]
graphs [
 ClearActiveContent {nodes [
   ClearActiveContent() {is=title K2Node_FunctionEntry}
   > SetHasActiveContent!(execute, self:Fort Prioritized Widget Interface=Self-Reference(), bNewActive:bool=false) -> (then)
  ]}
 EventGraph {nodes [
   n1 = OnAnimationsFinished!() -> (OutputDelegate, then, Name:Name) {is=title K2Node_CustomEvent}
   > Branch(Condition="Equal (Name)"!(A:Name=@n1.Name, B:Name=Get IntroAnimReverse() -> (IntroAnimReverse:Name=None)) -> (ReturnValue:bool) {is=title K2Node_CallFunction})
   > SetIsLocalChallengesVisible!(
    execute
    self:Fort HUDContext Object Reference=Get FortHUDContext!() -> (ReturnValue:Fort HUDContext Object Reference) {is=title K2Node_GetBlueprintContextEx}
    bIsVisible:bool=false
   ) -> (then)
   Event OnResetOutroDelay() {is=title K2Node_Event}
   > n2 = Branch(Condition=Is Valid Timer Handle!(Handle:Timer Handle Structure=Get ExpirationTimer() -> (ExpirationTimer:Timer Handle Structure)) -> (ReturnValue:bool))
   > StartOutroDelay()
   StartIntro(execute=@n2.else)
   Event Construct()
   > Set OriginalHeaderText(OriginalHeaderText:Text="GetText (Text)"!(self:Text Object Reference=Get Text_Challenges() -> (Text_Challenges:Common Text Object Reference)) -> (ReturnValue:Text)) -> (Output_Get:Text)
   > Bind Event to On Animations Finished(
    self:Fort Simple Widget Animations Panel Object Reference=Get SimpleAnimations() -> (SimpleAnimations:Fort Simple Widget Animations Panel Object Reference)
    Delegate=@n1
   ) {is=title K2Node_AddDelegate}
   n3 = HasChallengesToDisplay!() -> (bHasChallenges:bool)
   > Branch(Condition=OR Boolean!(A=IsCreativeOrPlaygroundModePlaylist!() -> (ReturnValue:bool), B=IsVkEditPlaylist!() -> (ReturnValue:bool), C:bool=NOT Boolean(A=@n3)) -> (ReturnValue))
   > SetVisibility(self=Get HeightClippingOverrideSB() -> (HeightClippingOverrideSB:Size Box Object Reference), InVisibility=Collapsed)
   n4 = Event OnNewContextualHeader!() -> (OutputDelegate, then, NewText:"Text (by ref)") {is=title K2Node_Event}
   > "SetText (Text)"(
    self=Get Text_Challenges()
    InText=Select(
     Option 0:"Text (by ref)"=Format Text!(Format="LOCTEXT(\"{poi name} Quests\")", poi name:"Text (by ref)"=@n4.NewText) -> (Result)
     Option 1:"Text (by ref)"=Get OriginalHeaderText() -> (OriginalHeaderText:Text)
     Index=TextIsEmpty(InText=@n4.NewText)
    ) -> (ReturnValue:"Text (by ref)")
   )
   Event BP_WidgetIsFinished() {is=title K2Node_Event}
   > FinishOutro()
   Event BP_OnWidgetActivated() {is=title K2Node_Event}
   > StartIntro()
  ]}
 FinishOutro {nodes [
   FinishOutro() {is=title K2Node_FunctionEntry}
   > Clear and Invalidate Timer by Handle(Handle=Get ExpirationTimer())
   > PlayAnimationByName!(
    execute
    self:Fort Simple Widget Animations Panel Object Reference=Get SimpleAnimations()
    Name:"Name (by ref)"=Get IntroAnimReverse() -> (IntroAnimReverse=None)
    Speed:float=1
    StartAtTime:float=0
   ) -> (then, ReturnValue:bool)
  ]}
 HasChallengesToDisplay {nodes [
   HasChallengesToDisplay() {is=title K2Node_FunctionEntry}
   n1 = Get DynamicEntryBox_ChallengeEntries() -> (DynamicEntryBox_ChallengeEntries:Dynamic Entry Box Object Reference)
   > Return Node!(
    execute
    bHasChallenges:bool="integer > integer"(A=GetNumEntries!(self:Dynamic Entry Box Base Object Reference=@n1) -> (ReturnValue:int=0), B=0) {is=title K2Node_CallFunction}
   )
  ]}
 StartIntro {nodes [
   StartIntro() {is=title K2Node_FunctionEntry}
   > Branch(Condition=HasChallengesToDisplay() {is=title K2Node_CallFunction})
   > SetVisibility(InVisibility=SelfHitTestInvisible)
   > SetIsLocalChallengesVisible(self=Get FortHUDContext(), bIsVisible=true)
   > Clear and Invalidate Timer by Handle(Handle=Get ExpirationTimer())
   n1 = GetAllEntries!(self:Dynamic Entry Box Base Object Reference=Get DynamicEntryBox_ChallengeEntries()) -> (ReturnValue:Array of User Widget Object References)
   > n2 = For Each Loop(Array:Array of User Widget Object References=@n1) -> (Array Element:User Widget Object Reference)
   > n3 = Cast To AthenaContextualChallengeEntry_BP!(execute, Object=@n2) -> (then, CastFailed, AsAthena Contextual Challenge Entry BP:Athena Contextual Challenge Entry BP Object Reference) {is=title K2Node_DynamicCast}
   n4 = Length(TargetArray:Array of User Widget Object References=@n1)
   > StyleLastElement!(
    execute
    self:Athena Contextual Challenge Entry BP Object Reference=@n3
    isLast:bool="Equal (Integer)"(A=@n2.Array Index, B=int - int(A=@n4, B=1) {is=title K2Node_CallFunction}) {is=title K2Node_CallFunction}
   ) -> (then)
   PlayAnimationByName(execute=@n2.Completed, self=Get SimpleAnimations(), Name=Get IntroAnim() -> (IntroAnim:Name=None), Speed=1, StartAtTime=0)
   > StartOutroDelay()
  ]}
 StartOutroDelay {nodes [
   StartOutroDelay() {is=title K2Node_FunctionEntry}
   > Branch(Condition=ShouldOutroAutomatically!(self:Athena Contextual Challenges Widget Object Reference) -> (ReturnValue:bool))
   > Clear and Invalidate Timer by Handle(Handle=Get ExpirationTimer())
   > n1 = Set Timer by Event(
    Delegate=Create Event()
    Time=GetOutroDelay!(self:Athena Contextual Challenges Widget Object Reference) -> (ReturnValue:float)
    bLooping=false
   )
   > Set ExpirationTimer(ExpirationTimer:Timer Handle Structure=@n1) -> (Output_Get:Timer Handle Structure)
  ]}
]
variables [
 ExpirationTimer {type=TimerHandle}
 IntroAnim {type=string, default=IntroAnim}
 IntroAnimReverse {type=string, default=IntroAnimReverse}
 OriginalHeaderText {type=string, default=""}
]
widgets {
 info {inheritedWidgetCount=2, parentClass=/Script/FortniteUI.AthenaContextualChallengesWidget, rootWidgetClass=&SizeBox, widgetCount=8}
 widgets [
  HeightClippingOverrideSB {
   widgetClassPath=&SizeBox
   bIsVariable=true
   widget=HeightClippingOverrideSB
   properties {bOverride_WidthOverride=true, clipping=ClipToBounds, maxDesiredHeight=320, widthOverride=448}
  }
  SimpleAnimations {
   widgetClassPath=/Script/FortniteUI.FortSimpleWidgetAnimationsPanel
   bIsVariable=true
   parent=HeightClippingOverrideSB
   widget=SimpleAnimations
   properties {animations [
     IntroAnim {
      bIsEnabled=true
      onAnimationsFinished=()
      targets [{
        alpha=&/IntroAnim_RenderOpacity
        bResetOnFinish=false
        bShouldAnimateAlpha=true
        bShouldAnimateScale=false
        bShouldAnimateTranslation=true
        scale=null
        targetName=VerticalBox_0
        translation=&/IntroAnim_Translation
        materialScalarAnimations []
       }]
     }
     IntroAnimReverse {
      bIsEnabled=true
      onAnimationsFinished=()
      targets [{
        alpha=&/IntroAnim_RenderOpacityReverse
        bResetOnFinish=false
        bShouldAnimateAlpha=true
        bShouldAnimateScale=false
        bShouldAnimateTranslation=true
        scale=null
        targetName=VerticalBox_0
        translation=&/IntroAnim_TranslationReverse
        materialScalarAnimations []
       }]
     }
    ]}
   slot {class=SizeBoxSlot}
  }
  VerticalBox_0 {widgetClassPath=&VerticalBox, parent=SimpleAnimations, widget=VerticalBox_0, slot {class=FortSimpleWidgetAnimationsSlot}}
  ScaleBox_0 {
   widgetClassPath=&ScaleBox
   parent=VerticalBox_0
   widget=ScaleBox_0
   slot {class=VerticalBoxSlot, properties {padding=<ltrb 4 0 0 0>}}
  }
  Text_Challenges {
   widgetClassPath=&CommonTextBlock
   bIsVariable=true
   parent=ScaleBox_0
   widget=Text_Challenges
   properties {
    bAutoCollapseWithEmptyText=true
    text=Local Quests
    textTransformPolicy=ToUpper
    font {
     bForceMonospaced=false
     bMaterialIsStencil=false
     fontMaterial=null
     fontObject=&BurbankSmall
     letterSpacing=0
     monospacedWidth=1
     size=18
     skewAmount=0
     typefaceFontName=Black
     outlineSettings {
      bApplyOutlineToDropShadows=false
      bMiteredCorners=false
      bSeparateFillAlpha=false
      outlineMaterial=&M-FontShadow
      outlineSize=3
      outlineColor=<rgba 0 0 0 0.5>
     }
    }
    renderTransform {angle=0, scale=<1 1>, shear=<-8 0>, translation=<0 0>}
    shadowOffset=<0 0>
   }
   slot {class=ScaleBoxSlot, properties {horizontalAlignment=HAlign_Fill}}
  }
  Separator {
   widgetClassPath=&Image
   parent=VerticalBox_0
   widget=Separator
   properties {
    visibility=HitTestInvisible
    brush {
     bIsDynamicallyLoaded=false
     drawAs=Box
     imageType=NoImage
     mirroring=NoMirror
     resourceName=None
     resourceObject=null
     tiling=NoTile
     imageSize=<16 2>
     margin=<ltrb 0.5 0 0.5 0>
     outlineSettings {
      bUseBrushTransparency=false
      roundingType=HalfHeightRadius
      width=0
      color {colorUseRule=UseColor_Specified, specifiedColor=<rgba 0 0 0 0>}
      cornerRadii=<0 0 0 0>
     }
     tintColor {colorUseRule=UseColor_Specified, specifiedColor=<rgba 1 1 1 0.4>}
     uVRegion {bIsValid=false, max=<0 0>, min=<0 0>}
    }
   }
   slot {class=VerticalBoxSlot, properties {padding=<ltrb 0 2 0 12>}}
  }
  DynamicEntryBox_ChallengeEntries {
   widgetClassPath=/Script/FortniteUI.FortDynamicEntryBox
   bInherited=true
   bIsVariable=true
   parent=VerticalBox_0
   widget=DynamicEntryBox_ChallengeEntries
   properties {
    entryBoxType=Vertical
    entryWidgetClass=&/AthenaContextualChallengeEntry_BP.AthenaContextualChallengeEntry_BP_C
    numDesignerPreviewEntries=5
    entrySpacing=<0 8>
   }
   slot {class=VerticalBoxSlot, properties {size {sizeRule=Fill, value=1}}}
  }
  TutorialHighlight {widgetClassPath=&TutorialHighlight}
 ]
}
