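# CommonUserWidget layout for a contextual tutorial tip callout on the HUD, with no logic beyond empty event stubs.
# A skewed dialog holds keybind-aware rich text, a "!" badge, a hold-to-close controller prompt and a mobile X button.
# Two hidden ContextTutorialBall widgets sit on separate canvases, one beside the dialog and one bottom-centered.
# Success visuals (checkmark, triangle bursts, flares, green borders) start transparent, as do both close controls.
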WidgetBlueprint /Game/Athena/HUD/Tutorial/HUD_ContextualTips_Base_Proxy
graphs [EventGraph {nodes [
   Event PreConstruct()
   Event Construct()
   Event Tick()
  ]}]
variables []
widgets {
 info {inheritedWidgetCount=0, parentClass=&CommonUserWidget, rootWidgetClass=&Overlay, widgetCount=53}
 widgets [
  ScreenOverlay {widgetClassPath=&Overlay, widget=ScreenOverlay}
  CanvasPanel_0 {
   widgetClassPath=&CanvasPanel
   parent=ScreenOverlay
   widget=CanvasPanel_0
   slot {class=OverlaySlot, properties {horizontalAlignment=HAlign_Center, verticalAlignment=VAlign_Bottom, padding=<ltrb 0 0 0 300>}}
  }
  CanvasPanel_Ball_Parented {
   widgetClassPath=&CanvasPanel
   bIsVariable=true
   parent=CanvasPanel_0
   widget=CanvasPanel_Ball_Parented
   slot {class=CanvasPanelSlot, properties=~}
  }
  ContextTutorialBallParented {
   widgetClassPath=/Game/Athena/Tutorial/Contextual/ContextTutorialBall
   bIsVariable=true
   parent=CanvasPanel_Ball_Parented
   widget=ContextTutorialBallParented
   properties {bCanArrowAnimationPlay=true, visibility=Hidden, elementTag {tagName=Athena.HideForCapture}}
   slot {
    class=CanvasPanelSlot
    properties {zOrder=1, layoutData {alignment=<0 0>, anchors {maximum=<0 0>, minimum=<0 0>}, offsets=<ltrb 0 0 100 100>}}
   }
  }
  CanvasPanel_Container {
   widgetClassPath=&CanvasPanel
   parent=ScreenOverlay
   widget=CanvasPanel_Container
   slot {class=OverlaySlot, properties {verticalAlignment=VAlign_Center, padding=<ltrb 0 0 0 50>}}
  }
  CanvasPanel_Ball {
   widgetClassPath=&CanvasPanel
   bIsVariable=true
   parent=CanvasPanel_Container
   widget=CanvasPanel_Ball
   slot {
    class=CanvasPanelSlot
    properties {zOrder=1, layoutData {alignment=<0 0>, anchors {maximum=<0.5 0.5>, minimum=<0.5 0.5>}, offsets=<ltrb 137.1564 -119.8024 100 100>}}
   }
  }
  ContextTutorialBall {
   widgetClassPath=&ContextTutorialBall
   bIsVariable=true
   parent=CanvasPanel_Ball
   widget=ContextTutorialBall
   properties {visibility=Hidden, elementTag {tagName=Athena.HideForCapture}}
   slot {
    class=CanvasPanelSlot
    properties {zOrder=1, layoutData {alignment=<0 0>, anchors {maximum=<0 0>, minimum=<0 0>}, offsets=<ltrb 0 0 100 100>}}
   }
  }
  Overlay_Container {
   widgetClassPath=&Overlay
   bIsVariable=true
   parent=CanvasPanel_Container
   widget=Overlay_Container
   properties {visibility=HitTestInvisible, renderTransform {angle=0, scale=<1 1>, shear=<0 0>, translation=<40 0>}}
   slot {
    class=CanvasPanelSlot
    properties {layoutData {alignment=<0 0>, anchors {maximum=<0 0>, minimum=<0 0>}, offsets=<ltrb 0 0 519.1558 219.4225>}}
   }
  }
  Overlay_5 {
   widgetClassPath=&Overlay
   parent=Overlay_Container
   widget=Overlay_5
   slot {class=OverlaySlot, properties {horizontalAlignment=HAlign_Fill, verticalAlignment=VAlign_Fill}}
  }
  Overlay_PositiveFeedback {
   widgetClassPath=&Overlay
   parent=Overlay_5
   widget=Overlay_PositiveFeedback
   properties {bIsEnabled=false, renderOpacity=0, visibility=Collapsed, renderTransform {angle=0, scale=<1 1>, shear=<-10 0>, translation=<0 0>}}
   slot {class=OverlaySlot, properties {horizontalAlignment=HAlign_Fill, verticalAlignment=VAlign_Fill}}
  }
  PositiveFeedback_BG {
   widgetClassPath=&Image
   parent=Overlay_PositiveFeedback
   widget=PositiveFeedback_BG
   properties {
    brush {
     bIsDynamicallyLoaded=false
     drawAs=Box
     imageType=NoImage
     mirroring=NoMirror
     resourceName=None
     resourceObject=/Game/Athena/HUD/LTM/Tutorial/Materials/M_UI_SkewBoxBorder_Inst
     tiling=NoTile
     imageSize=<64 64>
     margin=<ltrb 0.2 0.2 0.2 0.2>
     outlineSettings {
      bUseBrushTransparency=false
      roundingType=HalfHeightRadius
      width=0
      color {colorUseRule=UseColor_Specified, specifiedColor=<rgba 0 0 0 0>}
      cornerRadii=<0 0 0 0>
     }
     tintColor {colorUseRule=UseColor_Specified, specifiedColor=<rgba 0.0343 1 0.227 1>}
     uVRegion {bIsValid=false, max=<0 0>, min=<0 0>}
    }
    renderTransform {angle=0, scale=<1 1>, shear=<-5.0682 0>, translation=<0 0>}
   }
   slot {class=OverlaySlot, properties {horizontalAlignment=HAlign_Fill, verticalAlignment=VAlign_Fill}}
  }
  Overlay_PositiveFeedback_2 {
   widgetClassPath=&Overlay
   parent=Overlay_5
   widget=Overlay_PositiveFeedback_2
   properties {bIsEnabled=false, renderOpacity=0, visibility=Hidden, renderTransform {angle=0, scale=<1 1>, shear=<-10 0>, translation=<0 0>}}
   slot {class=OverlaySlot, properties {horizontalAlignment=HAlign_Fill, verticalAlignment=VAlign_Fill}}
  }
  PositiveFeedback_BG_1 {
   widgetClassPath=&Image
   parent=Overlay_PositiveFeedback_2
   widget=PositiveFeedback_BG_1
   properties {
    brush {
     bIsDynamicallyLoaded=false
     drawAs=Box
     imageType=NoImage
     mirroring=NoMirror
     resourceName=None
     resourceObject=&M_UI_SkewBoxBorder_Inst
     tiling=NoTile
     imageSize=<64 64>
     margin=<ltrb 0.2 0.2 0.2 0.2>
     outlineSettings {
      bUseBrushTransparency=false
      roundingType=HalfHeightRadius
      width=0
      color {colorUseRule=UseColor_Specified, specifiedColor=<rgba 0 0 0 0>}
      cornerRadii=<0 0 0 0>
     }
     tintColor {colorUseRule=UseColor_Specified, specifiedColor=<rgba 0.0343 1 0.227 1>}
     uVRegion {bIsValid=false, max=<0 0>, min=<0 0>}
    }
    renderTransform {angle=0, scale=<1 1>, shear=<-5.0682 0>, translation=<0 0>}
   }
   slot {class=OverlaySlot, properties {horizontalAlignment=HAlign_Fill, verticalAlignment=VAlign_Fill}}
  }
  Overlay_Dialog {
   widgetClassPath=&Overlay
   bIsVariable=true
   parent=Overlay_Container
   widget=Overlay_Dialog
   properties {renderTransformPivot=<0 0.6439>}
   slot {class=OverlaySlot, properties {verticalAlignment=VAlign_Center}}
  }
  OverlayTextContent {
   widgetClassPath=&Overlay
   parent=Overlay_Dialog
   widget=OverlayTextContent
   slot {class=OverlaySlot, properties {horizontalAlignment=HAlign_Fill, verticalAlignment=VAlign_Center}}
  }
  Animated_Close {
   widgetClassPath=&Overlay
   parent=OverlayTextContent
   widget=Animated_Close
   slot {class=OverlaySlot, properties {horizontalAlignment=HAlign_Fill, verticalAlignment=VAlign_Fill}}
  }
  TipForController {
   widgetClassPath=&Overlay
   parent=Animated_Close
   widget=TipForController
   properties {renderOpacity=0, renderTransform {angle=0, scale=<1 1>, shear=<0 0>, translation=<-20 92>}}
   slot {class=OverlaySlot, properties {horizontalAlignment=HAlign_Fill, verticalAlignment=VAlign_Bottom}}
  }
  HorizontalBox_2 {
   widgetClassPath=&HorizontalBox
   parent=TipForController
   widget=HorizontalBox_2
   slot {class=OverlaySlot, properties {horizontalAlignment=HAlign_Fill, verticalAlignment=VAlign_Bottom, padding=<ltrb 40 0 30 0>}}
  }
  TipActionWidget {
   widgetClassPath=&CommonActionWidget
   bIsVariable=true
   parent=HorizontalBox_2
   widget=TipActionWidget
   properties {
    progressMaterialParam=Percentage
    inputActions [{dataTable=/Game/Athena/Tutorial/Contextual/ContextTutorial_InputActionDataTable, rowName=CloseTip}]
    progressMaterialBrush {
     bIsDynamicallyLoaded=false
     drawAs=Image
     imageType=NoImage
     mirroring=NoMirror
     resourceName=None
     resourceObject=/Game/Athena/HUD/M_HealProgress
     tiling=NoTile
     imageSize=<80 80>
     margin=<ltrb 0 0 0 0>
     outlineSettings {
      bUseBrushTransparency=false
      roundingType=HalfHeightRadius
      width=0
      color {colorUseRule=UseColor_Specified, specifiedColor=<rgba 0 0 0 0>}
      cornerRadii=<0 0 0 0>
     }
     tintColor {colorUseRule=UseColor_Specified, specifiedColor=<rgba 1 1 1 1>}
     uVRegion {bIsValid=false, max=<0 0>, min=<0 0>}
    }
   }
   slot {class=HorizontalBoxSlot, properties {padding=<ltrb 0 0 15 0>}}
  }
  ControllerTip_Text {
   widgetClassPath=&CommonTextBlock
   bIsVariable=true
   parent=HorizontalBox_2
   widget=ControllerTip_Text
   properties {
    scrollStyle=/Game/Athena/Prototype/Blueprints/Galileo/UI/TextScrollStyle-Base-Ticker.TextScrollStyle-Base-Ticker_C
    text="(HOLD) Close Tip"
    font {
     bForceMonospaced=false
     bMaterialIsStencil=false
     fontMaterial=null
     fontObject=&BurbankSmall
     letterSpacing=0
     monospacedWidth=1
     size=20
     skewAmount=0
     typefaceFontName=Bold
     outlineSettings {
      bApplyOutlineToDropShadows=false
      bMiteredCorners=false
      bSeparateFillAlpha=false
      outlineMaterial=null
      outlineSize=0
      outlineColor=<rgba 0 0 0 1>
     }
    }
    shadowOffset=<0 0>
   }
   slot {class=HorizontalBoxSlot, properties {verticalAlignment=VAlign_Center, size {sizeRule=Fill, value=1}}}
  }
  OverlayBG {
   widgetClassPath=&Overlay
   parent=OverlayTextContent
   widget=OverlayBG
   slot {class=OverlaySlot, properties {horizontalAlignment=HAlign_Fill, verticalAlignment=VAlign_Fill}}
  }
  Image_BGDropShadow {
   widgetClassPath=&Image
   parent=OverlayBG
   widget=Image_BGDropShadow
   properties {
    brush {
     bIsDynamicallyLoaded=false
     drawAs=Box
     imageType=NoImage
     mirroring=NoMirror
     resourceName=None
     resourceObject=/Game/Athena/HUD/LTM/Tutorial/Materials/M_UI_SkewBox_TutorialBG1
     tiling=NoTile
     imageSize=<256 128>
     margin=<ltrb 0.2 0 0.2 0>
     outlineSettings {
      bUseBrushTransparency=false
      roundingType=HalfHeightRadius
      width=0
      color {colorUseRule=UseColor_Specified, specifiedColor=<rgba 0 0 0 0>}
      cornerRadii=<0 0 0 0>
     }
     tintColor {colorUseRule=UseColor_Specified, specifiedColor=<rgba 0.0052 0.0052 0.0056 0.4>}
     uVRegion {bIsValid=false, max=<0 0>, min=<0 0>}
    }
    renderTransform {angle=1.5, scale=<1 1>, shear=<-2 0>, translation=<13 13>}
    renderTransformPivot=<0.2 0.5>
   }
   slot {class=OverlaySlot, properties {horizontalAlignment=HAlign_Fill, verticalAlignment=VAlign_Fill}}
  }
  MobileCloseBUtton {
   widgetClassPath=&Overlay
   parent=OverlayBG
   widget=MobileCloseBUtton
   properties {renderOpacity=0, visibility=Visible, renderTransform {angle=0, scale=<1 1>, shear=<0 0>, translation=<61.9627 2>}}
   slot {class=OverlaySlot, properties {horizontalAlignment=HAlign_Right}}
  }
  SizeBox_2 {
   widgetClassPath=&SizeBox
   parent=MobileCloseBUtton
   widget=SizeBox_2
   properties {bOverride_HeightOverride=true, bOverride_WidthOverride=true, heightOverride=80, visibility=Visible, widthOverride=80}
   slot {class=OverlaySlot}
  }
  CloseButtonBaking {widgetClassPath=&Overlay, parent=SizeBox_2, widget=CloseButtonBaking, slot {class=SizeBoxSlot}}
  Image_BGCloseMobileButton {
   widgetClassPath=&Image
   parent=CloseButtonBaking
   widget=Image_BGCloseMobileButton
   properties {
    brush {
     bIsDynamicallyLoaded=false
     drawAs=Box
     imageType=NoImage
     mirroring=NoMirror
     resourceName=None
     resourceObject=&M_UI_SkewBox_TutorialBG1
     tiling=NoTile
     imageSize=<45 90>
     margin=<ltrb 0.2 0 0.2 0>
     outlineSettings {
      bUseBrushTransparency=false
      roundingType=HalfHeightRadius
      width=0
      color {colorUseRule=UseColor_Specified, specifiedColor=<rgba 0 0 0 0>}
      cornerRadii=<0 0 0 0>
     }
     tintColor {colorUseRule=UseColor_Specified, specifiedColor=<rgba 0 0.117 1 1>}
     uVRegion {bIsValid=false, max=<0 0>, min=<0 0>}
    }
    renderTransform {angle=1.5, scale=<1 1>, shear=<-2 0>, translation=<0 0>}
    renderTransformPivot=<0.2 0.5>
   }
   slot {class=OverlaySlot, properties {horizontalAlignment=HAlign_Fill, verticalAlignment=VAlign_Fill}}
  }
  Image_IconMobileButton {
   widgetClassPath=&Image
   parent=CloseButtonBaking
   widget=Image_IconMobileButton
   properties {
    brush {
     bIsDynamicallyLoaded=false
     drawAs=Image
     imageType=NoImage
     mirroring=NoMirror
     resourceName=None
     resourceObject=/Game/UI/Foundation/Textures/Icons/Checks/T-Icon-X-128
     tiling=NoTile
     imageSize=<44.3956 41.2982>
     margin=<ltrb 0.2 0 0.2 0>
     outlineSettings {
      bUseBrushTransparency=false
      roundingType=HalfHeightRadius
      width=0
      color {colorUseRule=UseColor_Specified, specifiedColor=<rgba 0 0 0 0>}
      cornerRadii=<0 0 0 0>
     }
     tintColor {colorUseRule=UseColor_Specified, specifiedColor=<rgba 1 1 1 1>}
     uVRegion {bIsValid=false, max=<0 0>, min=<0 0>}
    }
    renderTransform {angle=1.5, scale=<1 1>, shear=<-2 0>, translation=<0 0>}
    renderTransformPivot=<0.2 0.5>
   }
   slot {class=OverlaySlot, properties {horizontalAlignment=HAlign_Center, verticalAlignment=VAlign_Center}}
  }
  Image_BGCloseMobileButton_Shadow {
   widgetClassPath=&Image
   parent=CloseButtonBaking
   widget=Image_BGCloseMobileButton_Shadow
   properties {
    brush {
     bIsDynamicallyLoaded=false
     drawAs=Image
     imageType=NoImage
     mirroring=NoMirror
     resourceName=None
     resourceObject=&M_UI_SkewBox_TutorialBG1
     tiling=NoTile
     imageSize=<31 90>
     margin=<ltrb 0 0 0 0>
     outlineSettings {
      bUseBrushTransparency=false
      roundingType=HalfHeightRadius
      width=0
      color {colorUseRule=UseColor_Specified, specifiedColor=<rgba 0 0 0 0>}
      cornerRadii=<0 0 0 0>
     }
     tintColor {colorUseRule=UseColor_Specified, specifiedColor=<rgba 0 0 0 0.17>}
     uVRegion {bIsValid=false, max=<0 0>, min=<0 0>}
    }
    renderTransform {angle=1.5, scale=<1 1>, shear=<-5 0>, translation=<0 0>}
    renderTransformPivot=<0.2 0.5>
   }
   slot {class=OverlaySlot, properties {verticalAlignment=VAlign_Fill}}
  }
  Image_BGOutline {
   widgetClassPath=&Image
   parent=OverlayBG
   widget=Image_BGOutline
   properties {
    visibility=SelfHitTestInvisible
    brush {
     bIsDynamicallyLoaded=false
     drawAs=Box
     imageType=NoImage
     mirroring=NoMirror
     resourceName=None
     resourceObject=&M_UI_SkewBox_TutorialBG1
     tiling=NoTile
     imageSize=<256 128>
     margin=<ltrb 0.2 0 0.2 0>
     outlineSettings {
      bUseBrushTransparency=false
      roundingType=HalfHeightRadius
      width=0
      color {colorUseRule=UseColor_Specified, specifiedColor=<rgba 0 0 0 0>}
      cornerRadii=<0 0 0 0>
     }
     tintColor {colorUseRule=UseColor_Specified, specifiedColor=<rgba 1 1 1 1>}
     uVRegion {bIsValid=false, max=<0 0>, min=<0 0>}
    }
    renderTransform {angle=0, scale=<1 1>, shear=<-1 0>, translation=<0 0>}
    renderTransformPivot=<0 0.5>
   }
   slot {class=OverlaySlot, properties {horizontalAlignment=HAlign_Fill, verticalAlignment=VAlign_Fill}}
  }
  Image_BGInset {
   widgetClassPath=&Image
   parent=OverlayBG
   widget=Image_BGInset
   properties {
    visibility=SelfHitTestInvisible
    brush {
     bIsDynamicallyLoaded=false
     drawAs=Box
     imageType=NoImage
     mirroring=NoMirror
     resourceName=None
     resourceObject=&M_UI_SkewBox_TutorialBG1
     tiling=NoTile
     imageSize=<256 128>
     margin=<ltrb 0.2 0 0.3 0>
     outlineSettings {
      bUseBrushTransparency=false
      roundingType=HalfHeightRadius
      width=0
      color {colorUseRule=UseColor_Specified, specifiedColor=<rgba 0 0 0 0>}
      cornerRadii=<0 0 0 0>
     }
     tintColor {colorUseRule=UseColor_Specified, specifiedColor=<rgba 1 1 1 1>}
     uVRegion {bIsValid=false, max=<0 0>, min=<0 0>}
    }
    colorAndOpacity=<rgba 0.0273 0.0273 0.0273 1>
    renderTransform {angle=0, scale=<1 1>, shear=<0 0>, translation=<0 -1>}
   }
   slot {class=OverlaySlot, properties {horizontalAlignment=HAlign_Fill, verticalAlignment=VAlign_Fill, padding=<ltrb 4 4 5 5>}}
  }
  Image_BGInset_Corner {
   widgetClassPath=&Image
   parent=OverlayBG
   widget=Image_BGInset_Corner
   properties {
    visibility=SelfHitTestInvisible
    brush {
     bIsDynamicallyLoaded=false
     drawAs=Box
     imageType=NoImage
     mirroring=NoMirror
     resourceName=None
     resourceObject=/Game/Athena/HUD/LTM/Tutorial/Materials/M_UI_SkewBox_TutorialCornerBG
     tiling=NoTile
     imageSize=<256 128>
     margin=<ltrb 0 0 1 0>
     outlineSettings {
      bUseBrushTransparency=false
      roundingType=HalfHeightRadius
      width=0
      color {colorUseRule=UseColor_Specified, specifiedColor=<rgba 0 0 0 0>}
      cornerRadii=<0 0 0 0>
     }
     tintColor {colorUseRule=UseColor_Specified, specifiedColor=<rgba 1 1 1 1>}
     uVRegion {bIsValid=false, max=<0 0>, min=<0 0>}
    }
    colorAndOpacity=<rgba 0 0.117 1 1>
   }
   slot {class=OverlaySlot, properties {horizontalAlignment=HAlign_Fill, verticalAlignment=VAlign_Fill, padding=<ltrb 4 4 5 5>}}
  }
  Image_BGInset_Corner_Solo {
   widgetClassPath=&Image
   parent=OverlayTextContent
   widget=Image_BGInset_Corner_Solo
   properties {
    visibility=Collapsed
    brush {
     bIsDynamicallyLoaded=false
     drawAs=Box
     imageType=NoImage
     mirroring=NoMirror
     resourceName=None
     resourceObject=&M_UI_SkewBox_TutorialCornerBG
     tiling=NoTile
     imageSize=<256 128>
     margin=<ltrb 0 0 1 0>
     outlineSettings {
      bUseBrushTransparency=false
      roundingType=HalfHeightRadius
      width=0
      color {colorUseRule=UseColor_Specified, specifiedColor=<rgba 0 0 0 0>}
      cornerRadii=<0 0 0 0>
     }
     tintColor {colorUseRule=UseColor_Specified, specifiedColor=<rgba 1 1 1 1>}
     uVRegion {bIsValid=false, max=<0 0>, min=<0 0>}
    }
    colorAndOpacity=<rgba 0 0.117 1 1>
   }
   slot {class=OverlaySlot, properties {horizontalAlignment=HAlign_Fill, verticalAlignment=VAlign_Fill, padding=<ltrb 4 4 5 5>}}
  }
  BorderText {
   widgetClassPath=&Border
   parent=OverlayTextContent
   widget=BorderText
   properties {
    visibility=HitTestInvisible
    background {
     bIsDynamicallyLoaded=false
     drawAs=NoDrawType
     imageType=NoImage
     mirroring=NoMirror
     resourceName=None
     resourceObject=null
     tiling=NoTile
     imageSize=<32 32>
     margin=<ltrb 0 0 0 0>
     outlineSettings {
      bUseBrushTransparency=false
      roundingType=HalfHeightRadius
      width=0
      color {colorUseRule=UseColor_Specified, specifiedColor=<rgba 0 0 0 0>}
      cornerRadii=<0 0 0 0>
     }
     tintColor {colorUseRule=UseColor_Specified, specifiedColor=<rgba 1 1 1 1>}
     uVRegion {bIsValid=false, max=<0 0>, min=<0 0>}
    }
    padding=<ltrb 0 0 0 0>
   }
   slot {class=OverlaySlot, properties {verticalAlignment=VAlign_Fill, padding=<ltrb 50 25 75 30>}}
  }
  SizeBox_Message {
   widgetClassPath=&SizeBox
   parent=BorderText
   widget=SizeBox_Message
   properties {maxDesiredWidth=360, renderTransformPivot=<0.5 1>}
   slot {class=BorderSlot, properties {padding=<ltrb 0 0 0 0>}}
  }
  FortRichTextBlockKeybind_122 {
   widgetClassPath=/Script/FortniteUI.FortRichTextBlockKeybind
   bIsVariable=true
   parent=SizeBox_Message
   widget=FortRichTextBlockKeybind_122
   properties {
    autoWrapText=true
    keybindWidgetClass=/FortUILibrary/Input/KeybindWidget.KeybindWidget_C
    text="LOREM <ENEMY>ENEMY</> OR <ALLY>ALLY</> DOLOR AMIT <B>SEQUITOR</> <KEYWORD>KEYWORD</>!"
    textStyleSet=/Game/Athena/HUD/LTM/Tutorial/Materials/ContextualTutorialCalloutStyles
    textTransformPolicy=ToUpper
    wrapTextAt=385
    defaultTextStyleOverride {
     overflowPolicy=Clip
     transformPolicy=None
     colorAndOpacity {colorUseRule=UseColor_Specified, specifiedColor=<rgba 1 1 1 1>}
     font {
      bForceMonospaced=false
      bMaterialIsStencil=false
      fontMaterial=null
      fontObject=&NotoSans
      letterSpacing=0
      monospacedWidth=1
      size=24
      skewAmount=0
      typefaceFontName=NotoSans-Regular
      outlineSettings {
       bApplyOutlineToDropShadows=false
       bMiteredCorners=false
       bSeparateFillAlpha=false
       outlineMaterial=null
       outlineSize=0
       outlineColor=<rgba 0 0 0 1>
      }
     }
     highlightColor {colorUseRule=UseColor_Specified, specifiedColor=<rgba 0 0 0 1>}
     highlightShape {
      bIsDynamicallyLoaded=false
      drawAs=Image
      imageType=NoImage
      mirroring=NoMirror
      resourceName=None
      resourceObject=null
      tiling=NoTile
      imageSize=<32 32>
      margin=<ltrb 0 0 0 0>
      outlineSettings {
       bUseBrushTransparency=false
       roundingType=HalfHeightRadius
       width=0
       color {colorUseRule=UseColor_Specified, specifiedColor=<rgba 0 0 0 0>}
       cornerRadii=<0 0 0 0>
      }
      tintColor {colorUseRule=UseColor_Specified, specifiedColor=<rgba 1 1 1 1>}
      uVRegion {bIsValid=false, max=<0 0>, min=<0 0>}
     }
     selectedBackgroundColor {colorUseRule=UseColor_Foreground, specifiedColor=<rgba 1 0 1 1>}
     shadowColorAndOpacity=<rgba 0 0 0 1>
     shadowOffset=<0 0>
     strikeBrush {
      bIsDynamicallyLoaded=false
      drawAs=Image
      imageType=NoImage
      mirroring=NoMirror
      resourceName=None
      resourceObject=null
      tiling=NoTile
      imageSize=<32 32>
      margin=<ltrb 0 0 0 0>
      outlineSettings {
       bUseBrushTransparency=false
       roundingType=HalfHeightRadius
       width=0
       color {colorUseRule=UseColor_Specified, specifiedColor=<rgba 0 0 0 0>}
       cornerRadii=<0 0 0 0>
      }
      tintColor {colorUseRule=UseColor_Specified, specifiedColor=<rgba 1 1 1 1>}
      uVRegion {bIsValid=false, max=<0 0>, min=<0 0>}
     }
     underlineBrush {
      bIsDynamicallyLoaded=false
      drawAs=Image
      imageType=NoImage
      mirroring=NoMirror
      resourceName=None
      resourceObject=null
      tiling=NoTile
      imageSize=<32 32>
      margin=<ltrb 0 0 0 0>
      outlineSettings {
       bUseBrushTransparency=false
       roundingType=HalfHeightRadius
       width=0
       color {colorUseRule=UseColor_Specified, specifiedColor=<rgba 0 0 0 0>}
       cornerRadii=<0 0 0 0>
      }
      tintColor {colorUseRule=UseColor_Specified, specifiedColor=<rgba 1 1 1 1>}
      uVRegion {bIsValid=false, max=<0 0>, min=<0 0>}
     }
    }
    renderTransform {angle=0, scale=<1 1>, shear=<-10 0>, translation=<0 0>}
   }
   slot {class=SizeBoxSlot, properties {horizontalAlignment=HAlign_Left, verticalAlignment=VAlign_Top}}
  }
  Overlay_2 {
   widgetClassPath=&Overlay
   parent=OverlayTextContent
   widget=Overlay_2
   properties {renderTransform {angle=0, scale=<1.1 0.95>, shear=<0 0>, translation=<0 0>}, renderTransformPivot=<0.5 0>}
   slot {class=OverlaySlot, properties {horizontalAlignment=HAlign_Right}}
  }
  CommonTextBlock_40 {
   widgetClassPath=&CommonTextBlock
   parent=Overlay_2
   widget=CommonTextBlock_40
   properties {
    text="!"
    visibility=HitTestInvisible
    font {
     bForceMonospaced=false
     bMaterialIsStencil=false
     fontMaterial=null
     fontObject=&Burbank
     letterSpacing=0
     monospacedWidth=1
     size=50
     skewAmount=0
     typefaceFontName=Burbank-Black
     outlineSettings {
      bApplyOutlineToDropShadows=false
      bMiteredCorners=false
      bSeparateFillAlpha=false
      outlineMaterial=null
      outlineSize=0
      outlineColor=<rgba 0 0 0 1>
     }
    }
    renderTransform {angle=0, scale=<1 1>, shear=<-12 0>, translation=<-7 -18>}
    shadowColorAndOpacity=<rgba 0.0052 0.0052 0.0056 0.8>
    shadowOffset=<8 8>
   }
   slot {class=OverlaySlot, properties {horizontalAlignment=HAlign_Center, verticalAlignment=VAlign_Center}}
  }
  OverlayCheck {
   widgetClassPath=&Overlay
   parent=Overlay_2
   widget=OverlayCheck
   properties {visibility=HitTestInvisible, renderTransform {angle=0, scale=<1.6 1.6>, shear=<0 0>, translation=<-4 -32>}}
   slot {class=OverlaySlot, properties {horizontalAlignment=HAlign_Center, verticalAlignment=VAlign_Center}}
  }
  Overlaytriangles {
   widgetClassPath=&Overlay
   parent=OverlayCheck
   widget=Overlaytriangles
   properties {renderOpacity=0}
   slot {class=OverlaySlot, properties {horizontalAlignment=HAlign_Center, verticalAlignment=VAlign_Fill, padding=<ltrb -50 0 -50 0>}}
  }
  HorizontalBox {
   widgetClassPath=&HorizontalBox
   parent=Overlaytriangles
   widget=HorizontalBox
   properties {renderTransform {angle=0, scale=<1 1>, shear=<-12 0>, translation=<0 0>}}
   slot {class=OverlaySlot, properties {horizontalAlignment=HAlign_Fill, verticalAlignment=VAlign_Center}}
  }
  ImageTriangleL_1 {
   widgetClassPath=&Image
   parent=HorizontalBox
   widget=ImageTriangleL_1
   properties {
    brush {
     bIsDynamicallyLoaded=false
     drawAs=Image
     imageType=NoImage
     mirroring=NoMirror
     resourceName=None
     resourceObject=/Game/UI/HUD/Materials/M_FX_TriangleScroll-Burst
     tiling=NoTile
     imageSize=<200 40>
     margin=<ltrb 0 0 0 0>
     outlineSettings {
      bUseBrushTransparency=false
      roundingType=HalfHeightRadius
      width=0
      color {colorUseRule=UseColor_Specified, specifiedColor=<rgba 0 0 0 0>}
      cornerRadii=<0 0 0 0>
     }
     tintColor {colorUseRule=UseColor_Specified, specifiedColor=<rgba 0 1 0.1907 1>}
     uVRegion {bIsValid=false, max=<0 0>, min=<0 0>}
    }
    renderTransform {angle=0, scale=<-1 1>, shear=<0 0>, translation=<0 0>}
    renderTransformPivot=<0.5 1>
   }
   slot {class=HorizontalBoxSlot, properties {horizontalAlignment=HAlign_Right, verticalAlignment=VAlign_Center, padding=<ltrb -200 0 0 0>}}
  }
  Spacer {widgetClassPath=&Spacer, parent=HorizontalBox, widget=Spacer, slot {class=HorizontalBoxSlot}}
  ImageTriangleR_1 {
   widgetClassPath=&Image
   parent=HorizontalBox
   widget=ImageTriangleR_1
   properties {
    brush {
     bIsDynamicallyLoaded=false
     drawAs=Image
     imageType=NoImage
     mirroring=NoMirror
     resourceName=None
     resourceObject=&M_FX_TriangleScroll-Burst
     tiling=NoTile
     imageSize=<200 40>
     margin=<ltrb 0 0 0 0>
     outlineSettings {
      bUseBrushTransparency=false
      roundingType=HalfHeightRadius
      width=0
      color {colorUseRule=UseColor_Specified, specifiedColor=<rgba 0 0 0 0>}
      cornerRadii=<0 0 0 0>
     }
     tintColor {colorUseRule=UseColor_Specified, specifiedColor=<rgba 0 1 0.1907 1>}
     uVRegion {bIsValid=false, max=<0 0>, min=<0 0>}
    }
    renderTransformPivot=<0.5 1>
   }
   slot {class=HorizontalBoxSlot, properties {verticalAlignment=VAlign_Center, padding=<ltrb 0 0 -200 0>}}
  }
  HorizontalBox_0 {
   widgetClassPath=&HorizontalBox
   parent=Overlaytriangles
   widget=HorizontalBox_0
   properties {renderTransform {angle=0, scale=<1 1>, shear=<-12 0>, translation=<0 0>}}
   slot {class=OverlaySlot, properties {verticalAlignment=VAlign_Center}}
  }
  ImageTriangleL {
   widgetClassPath=&Image
   parent=HorizontalBox_0
   widget=ImageTriangleL
   properties {
    brush {
     bIsDynamicallyLoaded=false
     drawAs=Image
     imageType=NoImage
     mirroring=NoMirror
     resourceName=None
     resourceObject=/Game/UI/HUD/Materials/M_FX_TriangleSpiral
     tiling=NoTile
     imageSize=<200 64>
     margin=<ltrb 0 0 0 0>
     outlineSettings {
      bUseBrushTransparency=false
      roundingType=HalfHeightRadius
      width=0
      color {colorUseRule=UseColor_Specified, specifiedColor=<rgba 0 0 0 0>}
      cornerRadii=<0 0 0 0>
     }
     tintColor {colorUseRule=UseColor_Specified, specifiedColor=<rgba 0 1 0.1907 1>}
     uVRegion {bIsValid=false, max=<0 0>, min=<0 0>}
    }
    renderTransform {angle=0, scale=<-1 1>, shear=<0 0>, translation=<0 0>}
    renderTransformPivot=<0.5 1>
   }
   slot {class=HorizontalBoxSlot, properties {horizontalAlignment=HAlign_Right, verticalAlignment=VAlign_Center, padding=<ltrb -200 0 0 0>}}
  }
  Spacer_70 {widgetClassPath=&Spacer, parent=HorizontalBox_0, widget=Spacer_70, slot {class=HorizontalBoxSlot}}
  ImageTriangleR {
   widgetClassPath=&Image
   parent=HorizontalBox_0
   widget=ImageTriangleR
   properties {
    brush {
     bIsDynamicallyLoaded=false
     drawAs=Image
     imageType=NoImage
     mirroring=NoMirror
     resourceName=None
     resourceObject=&M_FX_TriangleSpiral
     tiling=NoTile
     imageSize=<200 64>
     margin=<ltrb 0 0 0 0>
     outlineSettings {
      bUseBrushTransparency=false
      roundingType=HalfHeightRadius
      width=0
      color {colorUseRule=UseColor_Specified, specifiedColor=<rgba 0 0 0 0>}
      cornerRadii=<0 0 0 0>
     }
     tintColor {colorUseRule=UseColor_Specified, specifiedColor=<rgba 0 1 0.1907 1>}
     uVRegion {bIsValid=false, max=<0 0>, min=<0 0>}
    }
    renderTransformPivot=<0.5 1>
   }
   slot {class=HorizontalBoxSlot, properties {verticalAlignment=VAlign_Center, padding=<ltrb 0 0 -200 0>}}
  }
  Overlay_CheckmarkContainer {
   widgetClassPath=&Overlay
   parent=OverlayCheck
   widget=Overlay_CheckmarkContainer
   properties {renderOpacity=0, visibility=HitTestInvisible}
   slot {class=OverlaySlot, properties {horizontalAlignment=HAlign_Center, verticalAlignment=VAlign_Center}}
  }
  Checkmark_Shadow {
   widgetClassPath=&Image
   bIsVariable=true
   parent=Overlay_CheckmarkContainer
   widget=Checkmark_Shadow
   properties {
    brush {
     bIsDynamicallyLoaded=false
     drawAs=Image
     imageType=NoImage
     mirroring=NoMirror
     resourceName=None
     resourceObject=/Game/Athena/HUD/Atlas/Bacchus/Confirm_Normal_L_Sprite
     tiling=NoTile
     imageSize=<50 50>
     margin=<ltrb 0 0 0 0>
     outlineSettings {
      bUseBrushTransparency=false
      roundingType=HalfHeightRadius
      width=0
      color {colorUseRule=UseColor_Specified, specifiedColor=<rgba 0 0 0 0>}
      cornerRadii=<0 0 0 0>
     }
     tintColor {colorUseRule=UseColor_Specified, specifiedColor=<rgba 0 0 0 0.8>}
     uVRegion {bIsValid=false, max=<0 0>, min=<0 0>}
    }
    colorAndOpacity=<rgba 1 1 1 0.8>
    renderTransform {angle=0, scale=<1 1>, shear=<0 0>, translation=<3 4>}
   }
   slot {class=OverlaySlot}
  }
  Checkmark {
   widgetClassPath=&Image
   bIsVariable=true
   parent=Overlay_CheckmarkContainer
   widget=Checkmark
   properties {brush {
     bIsDynamicallyLoaded=false
     drawAs=Image
     imageType=NoImage
     mirroring=NoMirror
     resourceName=None
     resourceObject=&Confirm_Normal_L_Sprite
     tiling=NoTile
     imageSize=<50 50>
     margin=<ltrb 0 0 0 0>
     outlineSettings {
      bUseBrushTransparency=false
      roundingType=HalfHeightRadius
      width=0
      color {colorUseRule=UseColor_Specified, specifiedColor=<rgba 0 0 0 0>}
      cornerRadii=<0 0 0 0>
     }
     tintColor {colorUseRule=UseColor_Specified, specifiedColor=<rgba 0.0343 1 0.227 1>}
     uVRegion {bIsValid=false, max=<0 0>, min=<0 0>}
    }}
   slot {class=OverlaySlot, properties {horizontalAlignment=HAlign_Center, verticalAlignment=VAlign_Center}}
  }
  Overlay_Flare {
   widgetClassPath=&Overlay
   parent=OverlayCheck
   widget=Overlay_Flare
   properties {visibility=HitTestInvisible, renderTransform {angle=0, scale=<0.8 0.8>, shear=<0 0>, translation=<0 0>}}
   slot {class=OverlaySlot, properties {horizontalAlignment=HAlign_Center, verticalAlignment=VAlign_Center}}
  }
  ImageSpinningSpikeball1 {
   widgetClassPath=&Image
   parent=Overlay_Flare
   widget=ImageSpinningSpikeball1
   properties {
    renderOpacity=0
    brush {
     bIsDynamicallyLoaded=false
     drawAs=Image
     imageType=NoImage
     mirroring=NoMirror
     resourceName=None
     resourceObject=/Game/Athena/HUD/Materials/Victory/M_Flare_Spikeball
     tiling=NoTile
     imageSize=<256 256>
     margin=<ltrb 0 0 0 0>
     outlineSettings {
      bUseBrushTransparency=false
      roundingType=HalfHeightRadius
      width=0
      color {colorUseRule=UseColor_Specified, specifiedColor=<rgba 0 0 0 0>}
      cornerRadii=<0 0 0 0>
     }
     tintColor {colorUseRule=UseColor_Specified, specifiedColor=<rgba 1 1 1 1>}
     uVRegion {bIsValid=false, max=<0 0>, min=<0 0>}
    }
    renderTransform {angle=13.5163, scale=<2 2>, shear=<0 0>, translation=<0 0>}
   }
   slot {class=OverlaySlot, properties {horizontalAlignment=HAlign_Center, verticalAlignment=VAlign_Center, padding=<ltrb -128 -128 -128 -128>}}
  }
  Image1Cross {
   widgetClassPath=&Image
   parent=Overlay_Flare
   widget=Image1Cross
   properties {
    renderOpacity=0
    brush {
     bIsDynamicallyLoaded=false
     drawAs=Image
     imageType=NoImage
     mirroring=NoMirror
     resourceName=None
     resourceObject=/Game/Athena/HUD/Materials/Victory/M_Cross_Flare
     tiling=NoTile
     imageSize=<128 128>
     margin=<ltrb 0 0 0 0>
     outlineSettings {
      bUseBrushTransparency=false
      roundingType=HalfHeightRadius
      width=0
      color {colorUseRule=UseColor_Specified, specifiedColor=<rgba 0 0 0 0>}
      cornerRadii=<0 0 0 0>
     }
     tintColor {colorUseRule=UseColor_Specified, specifiedColor=<rgba 1 1 1 1>}
     uVRegion {bIsValid=false, max=<0 0>, min=<0 0>}
    }
    renderTransform {angle=0, scale=<2 2>, shear=<0 0>, translation=<0 0>}
   }
   slot {class=OverlaySlot, properties {horizontalAlignment=HAlign_Center, verticalAlignment=VAlign_Center, padding=<ltrb -64 -64 -64 -64>}}
  }
 ]
}
