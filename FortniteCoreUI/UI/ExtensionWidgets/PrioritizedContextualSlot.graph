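# HUD extension slot container that shares one contextual area between event notifications, challenge previews and tutorial tips.
# A prioritized widget switcher root holds all three, with notifications and challenges set to queue behind when contesting at their own priorities.
# When an extension widget is created, the graph adds it to the same switcher and casts its slot to set fill alignment on both axes.
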WidgetBlueprint /FortniteCoreUI/UI/ExtensionWidgets/PrioritizedContextualSlot
graphs [EventGraph {nodes [
   n1 = Event OnExtensionWidgetCreated!() -> (OutputDelegate, then, InUserWidget:User Widget Object Reference, InExtensionSlot:EUIExtensionSlot Enum=Invalid) {is=title K2Node_Event}
   > n2 = AddChild(
    self=Get PrioSwitcher_CharacterProgress() -> (PrioSwitcher_CharacterProgress:Fort Prioritized Widget Switcher Object Reference)
    Content=@n1.InUserWidget
   )
   > n3 = Cast To WidgetSwitcherSlot!(execute, Object=@n2) -> (then, CastFailed, AsWidget Switcher Slot:Widget Switcher Slot Object Reference) {is=title K2Node_DynamicCast}
   > SetVerticalAlignment(self:Widget Switcher Slot Object Reference=@n3, InVerticalAlignment=VAlign_Fill)
   > SetHorizontalAlignment(self:Widget Switcher Slot Object Reference=@n3, InHorizontalAlignment=HAlign_Fill)
  ]}]
variables []
widgets {
 info {
  inheritedWidgetCount=0
  parentClass=/Script/FortniteUI.FortExtensionWidgetContainer
  rootWidgetClass=/Script/FortniteUI.FortPrioritizedWidgetSwitcher
  widgetCount=4
 }
 widgets [
  PrioSwitcher_CharacterProgress {
   widgetClassPath=&FortPrioritizedWidgetSwitcher
   bIsVariable=true
   widget=PrioSwitcher_CharacterProgress
   properties {activeWidgetIndex=2, transitionDuration=0.2, visibility=HitTestInvisible}
  }
  EventNotifications {
   widgetClassPath=/FortniteCoreUI/UI/InGame/EventNotifications/EventNotifications
   parent=PrioSwitcher_CharacterProgress
   widget=EventNotifications
   properties {prioritizationData {priority=Priority_2, whenContesting=QueueBehind}}
   slot {class=WidgetSwitcherSlot}
  }
  AthenaContextualChallengesWidget_BP {
   widgetClassPath=/Game/Athena/HUD/Discoverability/AthenaContextualChallengesWidget_BP
   parent=PrioSwitcher_CharacterProgress
   widget=AthenaContextualChallengesWidget_BP
   properties {
    elementTag {tagName=HUD.Athena.ContextualChallengePreview}
    hUDElementTag {gameplayTags [
      {tagName=Athena.HideForCapture}
      {tagName=HUD.Athena.HideForPostGameSpectate}
      {tagName=HUD.Athena.HideOnSpectate}
      {tagName=HUD.Athena.HideforRadial}
      {tagName=Athena.Creative.HideForLoading}
      {tagName=HUD.Hide.WhileLoading}
      {tagName=HUD.Athena.HideForPostGame}
      {tagName=HUD.Hide.DuringXpEverywhereUI}
     ]}
    prioritizedWidgetData {priority=Priority_4, whenContesting=QueueBehind}
   }
   slot {class=WidgetSwitcherSlot}
  }
  HUD_ContextualTips_Base_Proxy {
   widgetClassPath=/Game/Athena/HUD/Tutorial/HUD_ContextualTips_Base_Proxy
   parent=PrioSwitcher_CharacterProgress
   widget=HUD_ContextualTips_Base_Proxy
   properties {visibility=Collapsed}
   slot {class=WidgetSwitcherSlot}
  }
 ]
}
