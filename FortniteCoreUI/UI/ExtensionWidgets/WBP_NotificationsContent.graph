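# HUD extension container for upper-center notifications, stacking phase alerts, storm surge, countdown, event notifications and other overlays in one prioritized widget switcher.
# Registers as a player-specific host for the UpperCenter slot tag, adding or removing hosted content and created extension widgets in that switcher.
# For ForbiddenFruit products, listens for local player place changes; place 1 reprioritizes and offsets EventNotifications.
# Destruct unregisters the host and stops the listener.
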WidgetBlueprint /FortniteCoreUI/UI/ExtensionWidgets/WBP_NotificationsContent
graphs [
 EventGraph {nodes [
   Event Construct()
   n1 = Get UpperCenterSlotTag() -> (UpperCenterSlotTag:Gameplay Tag Structure)
   > RegisterWidgetHost!(execute, WidgetHost:Widget Object Reference, WidgetHostTag:"Gameplay Tag Structure (by ref)"=@n1, IsPlayerSpecific:bool=true) -> (then)
   > Branch(Condition=MatchesAnyTags!(
     TagOne:Gameplay Tag Structure=GetCurrentContextProductTag!() -> (ReturnValue:Gameplay Tag Structure)
     OtherContainer:"Gameplay Tag Container Structure (by ref)"=MakeLiteralGameplayTagContainer!(Value:Gameplay Tag Container Structure=(GameplayTags=((TagName="Product.ForbiddenFruit")))) -> (ReturnValue:Gameplay Tag Container Structure)
     bExactMatch:bool=false
    ) -> (ReturnValue:bool))
   > n2 = StartListeningToEvent!(
    execute:Exec
    Target:Gameplay Event Router Component Object Reference=GetGlobalEventRouter()
    EventType:Script Struct Object Reference=/Script/FortniteGame.OnLocalPlayerPlaceChangedEvent
    Context:Object Reference=GetFirstLocalFortPlayerController!(bIncludeNonSpectators:bool=true, bIncludeSpectators:bool=false) -> (ReturnValue:Fort Player Controller Object Reference)
    EventBubblingRule:EEventBubblingRule Enum
   ) -> (
    then:Exec
    AsyncTaskProxy:Async Action Start Listening To Event Object Reference
    OnEventReceived:Exec
    Payload:On Local Player Place Changed Event Structure
   )
   > Set OnPlayerPlaceChangedEventTaskRef(OnPlayerPlaceChangedEventTaskRef:Async Action Start Listening To Event Object Reference=@n2) -> (Output_Get:Async Action Start Listening To Event Object Reference)
   Branch(
    execute=@n2.OnEventReceived
    Condition="Equal (Integer)"(
     A=Break On Local Player Place Changed Event!(OnLocalPlayerPlaceChangedEvent:"On Local Player Place Changed Event Structure (by ref)"=@n2.Payload) -> (Place:int) {is=title K2Node_BreakStruct}
     B=1
    )
   )
   n3 = Get EventNotifications() -> (EventNotifications:Event Notifications Object Reference)
   > SetNewWidgetPriority!(execute, self:Fort Game Event Notifications Object Reference=@n3, NewPriority:EFortPrioritizedWidgetPriority Enum=Priority_2) -> (then)
   > SetPadding!(
    execute
    self:User Widget Object Reference=@n3
    InPadding_Left:float=0
    InPadding_Top:float=0
    InPadding_Right:float=1500
    InPadding_Bottom:float=0
   ) -> (then)
   > SetAlignmentInViewport!(execute, self:User Widget Object Reference, Alignment:Vector 2D Structure) -> (then)
   n4 = Event Add Content To Host Widget!() -> (
    OutputDelegate
    then
    WidgetHostTag:"Gameplay Tag Structure (by ref)"
    ContentWidget:Widget Object Reference
    Padding:"Margin Structure (by ref)"
    bFill:bool
   ) {is=title K2Node_Event}
   > Branch(Condition=MatchesTag!(TagOne:Gameplay Tag Structure=@n4.WidgetHostTag, TagTwo:Gameplay Tag Structure=Get UpperCenterSlotTag(), bExactMatch:bool=true) -> (ReturnValue:bool))
   > AddChild(self=Get WidgetSwitcher_TopMid() -> (WidgetSwitcher_TopMid:Fort Prioritized Widget Switcher Object Reference), Content=@n4.ContentWidget)
   n5 = Event OnExtensionWidgetCreated!() -> (OutputDelegate, then, InUserWidget:User Widget Object Reference, InExtensionSlot:EUIExtensionSlot Enum=Invalid) {is=title K2Node_Event}
   > AddChild(self=Get WidgetSwitcher_TopMid(), Content=@n5.InUserWidget)
   Event Destruct()
   > UnregisterWidgetHost!(execute, WidgetHost:Widget Object Reference, WidgetHostTag:"Gameplay Tag Structure (by ref)"=@n1) -> (then)
   > StopListeningToEvent(self:Async Action Start Listening To Event Object Reference=Get OnPlayerPlaceChangedEventTaskRef() -> (OnPlayerPlaceChangedEventTaskRef:Async Action Start Listening To Event Object Reference))
   Event PreConstruct()
  ]}
 RemoveContentFromWidgetHost {nodes [
   n1 = Remove Content From Host Widget!() -> (
    then
    WidgetHostTag:"Gameplay Tag Structure (by ref)"
    ContentWidget:Widget Object Reference
    RemovalType:EWidgetSlottingRemovalType Enum=Unparent
   ) {is=title K2Node_FunctionEntry}
   > n2 = Branch(Condition=MatchesTag(TagOne=@n1, TagTwo=Get UpperCenterSlotTag(), bExactMatch=true))
   > n3 = RemoveChild(self=Get WidgetSwitcher_TopMid(), Content=@n1.ContentWidget)
   > Return Node!(execute, ReturnValue:bool=@n3)
   Return Node(execute=@n2.else, ReturnValue=false)
  ]}
]
variables [
 OnPlayerPlaceChangedEventTaskRef {type=AsyncAction_StartListeningToEvent, default=null}
 UpperCenterSlotTag {type=GameplayTag, default {tagName=DynamicUI.HUD.Fort.Athena.UpperCenter.Slot}}
]
widgets {
 info {inheritedWidgetCount=0, parentClass=/Script/FortniteUI.FortExtensionWidgetContainer, rootWidgetClass=&VerticalBox, widgetCount=11}
 widgets [
  TopMidVerticalBox {widgetClassPath=&VerticalBox, bIsVariable=true, widget=TopMidVerticalBox}
  WidgetSwitcher_TopMid {
   widgetClassPath=/Script/FortniteUI.FortPrioritizedWidgetSwitcher
   bIsVariable=true
   parent=TopMidVerticalBox
   widget=WidgetSwitcher_TopMid
   slot {class=VerticalBoxSlot, properties {horizontalAlignment=HAlign_Center, verticalAlignment=VAlign_Top}}
  }
  GamePhaseAlert {
   widgetClassPath=/Game/Athena/HUD/Phase/AthenaGamePhaseChangeWidget
   bIsVariable=true
   parent=WidgetSwitcher_TopMid
   widget=GamePhaseAlert
   properties {
    visibility=Collapsed
    elementTag {tagName=HUD.Athena.GamePhase.AlertMessage}
    hUDElementTag {gameplayTags [
      {tagName=Athena.HideForCapture}
      {tagName=HUD.LTM.HideForTryHard}
      {tagName=HUD.Event.HideForSpecialEvent}
      {tagName=Hud.Athena.HideOnCinematic}
      {tagName=HUD.Athena.HideForPostGame}
     ]}
    prioritizationData {priority=Priority_2, whenContesting=QueueBehind}
   }
   slot {class=WidgetSwitcherSlot, properties {verticalAlignment=VAlign_Top, padding=<ltrb 0 105 0 0>}}
  }
  AthenaStormSurgePhase {
   widgetClassPath=/Game/Athena/HUD/AthenaStormSurgePhase
   bIsVariable=true
   parent=WidgetSwitcher_TopMid
   widget=AthenaStormSurgePhase
   properties {
    elementTag {tagName=HUD.Special.StormSurgePhase}
    hUDElementTag {gameplayTags [{tagName=Hud.Athena.HideOnCinematic}, {tagName=HUD.Athena.HideForPostGame}, {tagName=HUD.Athena.HideForPostGameSpectate}]}
    prioritizationData {priority=Priority_2, whenContesting=QueueBehind}
   }
   slot {class=WidgetSwitcherSlot, properties {verticalAlignment=VAlign_Top}}
  }
  AthenaCountdown {
   widgetClassPath=/Game/Athena/HUD/AthenaCountdown
   bIsVariable=true
   parent=WidgetSwitcher_TopMid
   widget=AthenaCountdown
   properties {
    elementTag {tagName=HUD.Special.Countdown}
    hUDElementTag {gameplayTags [{tagName=Hud.Athena.HideOnCinematic}, {tagName=HUD.Athena.HideForPostGame}, {tagName=HUD.Athena.HideForPostGameSpectate}]}
    prioritizationData {priority=Priority_2, whenContesting=StompOther}
    renderTransform {angle=0, scale=<1 1>, shear=<0 0>, translation=<0 30>}
   }
   slot {class=WidgetSwitcherSlot, properties {horizontalAlignment=HAlign_Center, verticalAlignment=VAlign_Top}}
  }
  VolumePromptWidget {
   widgetClassPath=/Game/UI/Foundation/HUDTagPrompt/HUDTagPromptWidget
   bIsVariable=true
   parent=WidgetSwitcher_TopMid
   widget=VolumePromptWidget
   properties {hUDElementTag {gameplayTags [{tagName=HUD.LTM.HideForTryHard}, {tagName=Athena.HideForCapture}]}}
   slot {class=WidgetSwitcherSlot, properties {verticalAlignment=VAlign_Top}}
  }
  AthenaSpecialEventStartedOverlay {
   widgetClassPath=/Game/Athena/HUD/Discoverability/AthenaSpecialEventStartedOverlay
   bIsVariable=true
   parent=WidgetSwitcher_TopMid
   widget=AthenaSpecialEventStartedOverlay
   properties {elementTag {tagName=HUD.Special.SpecialEventStarted}, renderTransform {angle=0, scale=<0.85 0.85>, shear=<0 0>, translation=<0 0>}}
   slot {class=WidgetSwitcherSlot, properties {verticalAlignment=VAlign_Top}}
  }
  AthenaDuelOverlay {
   widgetClassPath=/Game/Athena/HUD/Duel/AthenaDuelOverlay
   bIsVariable=true
   parent=WidgetSwitcher_TopMid
   widget=AthenaDuelOverlay
   properties {elementTag {tagName=HUD.Athena.DuelOverlay}}
   slot {class=WidgetSwitcherSlot, properties {verticalAlignment=VAlign_Top}}
  }
  BP_AthenaCollectionItemToast {
   widgetClassPath=/Game/Athena/HUD/Collections/BP_AthenaCollectionItemToast
   parent=WidgetSwitcher_TopMid
   widget=BP_AthenaCollectionItemToast
   properties {hUDElementTag {gameplayTags [{tagName=Athena.HideForCapture}, {tagName=HUD.Athena.HideForPostGameSpectate}, {tagName=HUD.Athena.HideForPostGame}]}}
   slot {class=WidgetSwitcherSlot, properties {horizontalAlignment=HAlign_Center, verticalAlignment=VAlign_Top}}
  }
  EventNotifications {
   widgetClassPath=/FortniteCoreUI/UI/InGame/EventNotifications/EventNotifications
   bIsVariable=true
   parent=WidgetSwitcher_TopMid
   widget=EventNotifications
   properties {hUDElementTag {gameplayTags [
      {tagName=Athena.HideForCapture}
      {tagName=HUD.Event.HideForSpecialEvent}
      {tagName=HUD.Athena.HideforRadial}
      {tagName=HUD.Athena.HideForPostGameSpectate}
     ]}}
   slot {class=WidgetSwitcherSlot, properties {verticalAlignment=VAlign_Top}}
  }
  AthenaContextualChallengesWidget_BP {
   widgetClassPath=/Game/Athena/HUD/Discoverability/AthenaContextualChallengesWidget_BP
   bIsVariable=true
   parent=WidgetSwitcher_TopMid
   widget=AthenaContextualChallengesWidget_BP
   properties {
    elementTag {tagName=HUD.Athena.ContextualChallengePreview}
    hUDElementTag {gameplayTags [
      {tagName=Athena.HideForCapture}
      {tagName=HUD.Athena.HideForPostGameSpectate}
      {tagName=HUD.Athena.HideOnSpectate}
      {tagName=HUD.Athena.HideforRadial}
      {tagName=Athena.Creative.HideForLoading}
      {tagName=HUD.Hide.WhileLoading}
      {tagName=HUD.Athena.HideForPostGame}
     ]}
    prioritizedWidgetData {priority=Priority_4, whenContesting=QueueBehind}
   }
   slot {class=WidgetSwitcherSlot, properties {horizontalAlignment=HAlign_Center, verticalAlignment=VAlign_Top}}
  }
 ]
}
